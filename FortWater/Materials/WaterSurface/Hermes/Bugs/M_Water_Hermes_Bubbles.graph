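# Translucent lit bubble material for the Hermes water surface, rendering cards as fake hemispheres.
# Vertex math collapses local positions, rebuilds camera-facing cards and pulls them along the view ray into a vertically flattened WPO.
# Interpolated card offsets yield hemisphere opacity and a normal bent by the camera vector, feeding edge masks and blends.
# A separate chain projects camera-relative positions onto the actor's height plane; material output connections are not shown.
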Material /FortWater/Materials/WaterSurface/Hermes/Bugs/M_Water_Hermes_Bubbles
graph {
 properties {
  bAutomaticallySetUsageInEditor=false
  bTangentSpaceNormal=false
  blendMode=BLEND_Translucent
  opacityMaskClipValue=0.33
  translucencyLightingMode=TLM_SurfacePerPixelLighting
  usedShadingModels=Default Lit
 }
 nodes [
  Constant3Vector() {constant=<rgba 50 50 10 0>}
  n1 = Divide(A=VertexInterpolator(VS=NamedRerouteUsage()).PS) {constB=50}
  MaterialExpressionMultiply_13 = Multiply(A=@n1, B=@n1)
  n2 = Convert(None=@MaterialExpressionMultiply_13) {
   nodeName=Break Float3
   convertInputs [{
     type=Vector3
     expressionInput {
      expression=&self:MaterialExpressionMultiply_13
      inputName=None
      mask=0
      maskA=0
      maskB=0
      maskG=0
      maskR=0
      outputIndex=0
     }
    }]
   convertOutputs [{type=Scalar}, {type=Scalar}, {type=Scalar}]
  }
  n3 = Add(A=Add(A=@n2, B=@n2), B=@n2)
  n4 = SquareRoot(None=OneMinus(None=Saturate(None=@n3)))
  n5 = Add(A=Multiply(A=NamedRerouteDeclaration(None=@n4) {name=Hemisphere Opacity, nodeColor=<rgba 1 0 0.8203 1>}, B=CameraVectorWS()), B=@n1)
  n6 = Saturate(None=Multiply(A=ComponentMask(None=@n5) {b=true}) {constB=1.2})
  n7 = OneMinus(None=Multiply(A=@n6, B=@n6))
  n8 = Multiply(A=@n7, B=@n7)
  LinearInterpolate(A=Constant3Vector() {constant=<rgba 0 0 1 0>}, B=@n5, Alpha=OneMinus(None=@n8))
  LinearInterpolate(A=Constant3Vector() {constant=<rgba 0.1698 0.0683 0.1771 1>}, B=Constant3Vector() {constant=<rgba 0.3005 0.1119 0.314 1>})
  n9 = Saturate(None=Multiply(A=ComponentMask(None=@n5) {b=true}) {constB=3})
  n10 = OneMinus(None=Multiply(A=@n9, B=@n9))
  n11 = NamedRerouteUsage()
  Multiply(A=OneMinus(None=Multiply(A=@n10, B=@n10)), B=Saturate(None=Multiply(A=@n11) {constB=2}))
  n12 = ScalarParameter() {defaultValue=1}
  Multiply(A=ComponentMask(None=@n5) {g=true, r=true}, B=Add(A=Multiply(A=@n8, B=@n12)))
  n13 = Transform(None=WorldPosition() {worldPositionShaderOffset=WPT_CameraRelative}.XYZ) {transformSourceType=TRANSFORMSOURCE_World, transformType=TRANSFORM_Local}
  n14 = NamedRerouteDeclaration(None=@n13) {name=Local Camera Position, nodeColor=<rgba 0.1016 1 0 1>}
  n15 = NamedRerouteDeclaration(None=Multiply(A=LocalPosition().XYZ) {constB=-1}) {name=Crunched Cards, nodeColor=<rgba 1 0 0.5625 1>}
  NamedRerouteDeclaration(None=Normalize(VectorInput=Add(A=@n14, B=@n15))) {name=Card Vector, nodeColor=<rgba 1 0.3984 0 1>}
  n16 = NamedRerouteUsage()
  n17 = Normalize(VectorInput=CrossProduct(A=Constant3Vector() {constant=<rgba 0 0 1 0>}, B=@n16))
  n18 = LocalPosition()
  n19 = Multiply(A=@n17, B=ComponentMask(None=@n18.XY) {r=true})
  n20 = Multiply(A=CrossProduct(A=@n17, B=@n16), B=ComponentMask(None=@n18.XY) {g=true})
  n21 = Add(A=NamedRerouteDeclaration(None=Add(A=@n19, B=@n20)) {name=Local Offsets, nodeColor=<rgba 0 0.6719 1 1>}, B=NamedRerouteUsage())
  n22 = Add(A=@n21, B=NamedRerouteUsage())
  n23 = Add(A=@n21, B=Multiply(A=Normalize(VectorInput=@n22)) {constB=-50})
  NamedRerouteDeclaration(None=Multiply(A=Transform(None=@n23) {transformSourceType=TRANSFORMSOURCE_Local}, B=Constant3Vector() {constant=<rgba 1 1 0.8 0>})) {name=WPO, nodeColor=<rgba 0.4766 0 1 1>}
  NamedRerouteUsage()
  Reroute(None=@n13)
  n24 = WorldPosition() {worldPositionShaderOffset=WPT_CameraRelative}
  n25 = ActorPositionWS() {originType=CameraRelative}
  n26 = ComponentMask(None=@n25) {b=true}
  Subtract(A=Multiply(A=@n24.XYZ, B=Divide(A=@n26, B=@n24.Z)), B=@n25)
 ]
}
