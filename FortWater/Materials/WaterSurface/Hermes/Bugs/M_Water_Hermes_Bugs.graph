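# Two-sided masked Single Layer Water material for the Hermes Bugs water surface, with time-of-day colors and emission.
# Sine-animated offset textures form flowing patterns, and WaterBodyData supplies depth and height data.
# Scene depth builds edge foam, while parallax-offset cellular bubbles with sprite lookup feed absorption and normals at Epic quality only.
# Surface, foam, distant and underwater attributes are blended, with static switches for HLOD and static meshes.
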Material /FortWater/Materials/WaterSurface/Hermes/Bugs/M_Water_Hermes_Bugs
graph {
 properties {
  allowTranslucentCustomDepthWrites=true
  bTangentSpaceNormal=false
  bUseMaterialAttributes=true
  bUsedWithInstancedStaticMeshes=true
  bUsedWithNanite=true
  bUsedWithSplineMeshes=true
  bUsedWithWater=true
  blendMode=BLEND_Masked
  pixelDepthOffsetMode=PDOM_Legacy
  refractionMethod=RM_PixelNormalOffset
  shadingModel=MSM_SingleLayerWater
  twoSided=true
  usedShadingModels=SingleLayerWater
 }
 nodes [
  n1 = Constant()
  n2 = Add(A=WorldPosition().XY, B=Multiply(A=NamedRerouteUsage(), B=ScalarParameter() {defaultValue=5, parameterName=Bubbles Offset}) {constB=0.15}) {constB=0}
  n3 = ScalarParameter() {defaultValue=300, parameterName=Bubbles Tilling}
  n4 = TextureSample(UVs=ComponentMask(None=Divide(A=@n2, B=@n3) {constB=4}) {g=true, r=true}) {samplerType=SAMPLERTYPE_LinearColor, texture=&/T_Water_Hermes_Bugs_Bubbles}
  n5 = Add(A=@n4.B, B=Multiply(A=Time(), B=ScalarParameter() {defaultValue=0.18, parameterName=Bubbles Speed}) {constB=0.5})
  n6 = OneMinus(None=NamedRerouteDeclaration(None=Frac(None=@n5)) {name=Bubble Depth Mask, nodeColor=<rgba 0 1 0.6953 1>})
  n7 = Saturate(None=LinearInterpolate(Alpha=@n6) {constA=-0.2})
  n8 = ScalarParameter() {defaultValue=500, parameterName=Bubbles Depth}
  n9 = ComponentMask(None=Multiply(A=Multiply(A=@n7, B=@n8) {constB=500}, B=CameraVectorWS()) {constB=100}) {g=true, r=true}
  n10 = TextureSample(UVs=Divide(A=Subtract(A=@n2, B=@n9), B=ScalarParameter() {defaultValue=300, parameterName=Bubbles Tilling}) {constB=200}) {samplerType=SAMPLERTYPE_LinearColor, texture=&T_Water_Hermes_Bugs_Bubbles}
  n11 = OneMinus(None=NamedRerouteUsage())
  n12 = Multiply(A=Add(A=ComponentMask(None=@n10.RGB) {g=true, r=true}) {constB=-0.5}, B=Add(A=Multiply(A=@n11) {constB=1.5}) {constB=1.2})
  n13 = Multiply(A=@n12) {constB=2.5}
  MaterialExpressionMultiply_41 = Multiply(A=@n13, B=@n13)
  n14 = Convert(None=@MaterialExpressionMultiply_41) {
   nodeName=Break Float2
   convertInputs [{
     type=Vector2
     expressionInput {
      expression=&self:MaterialExpressionMultiply_41
      inputName=None
      mask=0
      maskA=0
      maskB=0
      maskG=0
      maskR=0
      outputIndex=0
     }
    }]
   convertOutputs [{type=Scalar}, {type=Scalar}]
  }
  n15 = Add(A=@n14, B=@n14)
  n16 = SquareRoot(None=OneMinus(None=Saturate(None=@n15)))
  n17 = Multiply(A=@n16) {constB=4}
  n18 = NamedRerouteDeclaration(None=Multiply(A=Saturate(None=@n17), B=NamedRerouteUsage())) {name=Bubble Mask, nodeColor=<rgba 0 1 0.8125 1>}
  n19 = CameraVectorWS()
  n20 = CrossProduct(A=@n19, B=Constant3Vector() {constant=<rgba 0 0 1 0>})
  n21 = DotProduct(A=@n12, B=Normalize(VectorInput=ComponentMask(None=@n20) {g=true, r=true}))
  n22 = DotProduct(A=@n12, B=Normalize(VectorInput=ComponentMask(None=@n19) {g=true, r=true}))
  n23 = TextureSample(UVs=Add(A=AppendVector(A=@n21, B=@n22)) {constB=0.5}) {
   samplerSource=SSM_Clamp_WorldGroupSettings
   samplerType=SAMPLERTYPE_LinearColor
   texture=&/T_Water_Hermes_Bugs_Bubbles_Sprite
  }
  n24 = Multiply(A=Add(A=@n23.R) {constB=-0.5}) {constB=10}
  n25 = OneMinus(None=NamedRerouteUsage())
  n26 = Multiply(A=StaticSwitchParameter(True=Constant(), False=Multiply(A=@n18, B=@n24)) {parameterName=HLOD}, B=@n25)
  n27 = CameraVectorWS()
  n28 = Saturate(None=Multiply(A=ComponentMask(None=@n27) {b=true}) {constB=4})
  NamedRerouteDeclaration(None=QualitySwitch(Default=@n1, Low=@n1, High=@n1, Medium=@n1, Epic=Multiply(A=@n26, B=@n28))) {name=Bubbles Absorption, nodeColor=<rgba 1 0 0.1875 1>}
  n29 = SquareRoot(None=@n6)
  NamedRerouteDeclaration(None=Multiply(A=@n10.A, B=Multiply(A=@n4.A, B=@n29))) {name=Cell Mask, nodeColor=<rgba 1 0.7031 0 1>}
  n30 = MaterialFunctionCall!(Texture, Position) {materialFunction=/FortWater/Materials/Functions/WaterBodyData}
  NamedRerouteDeclaration(None=@n30.Water Depth) {name=Water Depth, nodeColor=<rgba 0 0.0152 0.1095 1>}
  NamedRerouteDeclaration(None=@n30.Water Z Height) {name=Water Z Height, nodeColor=<rgba 0 1 0.2266 1>}
  NamedRerouteDeclaration(None=@n30.Terrain Z Height) {name=Terrain Z Height, nodeColor=<rgba 1 0 0.0938 1>}
  n31 = NamedRerouteUsage()
  n32 = WorldPosition()
  NamedRerouteDeclaration(None=StaticSwitchParameter(True=Constant(), False=Subtract(A=@n31, B=@n32.Z)) {parameterName=IsStaticMesh}) {name=WPO, nodeColor=<rgba 0 0.4141 1 1>}
  n33 = Constant()
  n34 = CameraVectorWS()
  n35 = Multiply(A=Subtract(A=SceneDepthWithoutWater!(UVs).R, B=PixelDepth().R), B=Abs(None=ComponentMask(None=@n34) {b=true}))
  n36 = PixelDepth()
  n37 = Saturate(None=Divide(A=Subtract(A=@n36.R) {constB=5000}) {constB=10000})
  n38 = StaticSwitchParameter(
   True=@n35
   False=LinearInterpolate(A=@n35, B=NamedRerouteUsage(), Alpha=NamedRerouteDeclaration(None=@n37) {name=Pixel Depth, nodeColor=<rgba 0.2891 1 0 1>})
  ) {parameterName=IsStaticMesh}
  n39 = ScalarParameter() {defaultValue=40, parameterName=Foam Range}
  n40 = OneMinus(None=Saturate(None=Divide(A=@n38, B=@n39) {constB=44}))
  n41 = Constant() {r=1}
  n42 = OneMinus(None=Saturate(None=Divide(A=@n38) {constB=15}))
  n43 = OneMinus(None=Multiply(A=@n42, B=@n42))
  n44 = Multiply(A=@n43, B=@n43)
  n45 = Multiply(
   A=NamedRerouteDeclaration(None=StaticSwitchParameter(True=@n33, False=@n40) {parameterName=HLOD}) {name=Edge Mask, nodeColor=<rgba 1 0.6699 0 1>}
   B=NamedRerouteDeclaration(None=StaticSwitchParameter(True=@n41, False=@n44) {parameterName=HLOD}) {name=Edge Fade, nodeColor=<rgba 0.0781 0 1 1>}
  )
  n46 = Divide(A=WorldPosition().XY, B=ScalarParameter() {defaultValue=500, parameterName=Foam Scale}) {constB=800}
  n47 = NamedRerouteUsage()
  n48 = Multiply(A=Add(A=ComponentMask(None=@n47) {g=true, r=true}) {constB=-0.5}) {constB=0.1}
  n49 = TextureSample(UVs=NamedRerouteDeclaration(None=Add(A=@n46, B=@n48)) {name=UVs, nodeColor=<rgba 1 0 0.1641 1>}) {samplerType=SAMPLERTYPE_LinearColor, texture=&/T_Water_Hermes_Foam_M}
  n50 = NamedRerouteUsage()
  n51 = ScalarParameter() {defaultValue=0.2, parameterName=Mask Offset}
  n52 = Multiply(A=Add(A=OneMinus(None=@n49.R), B=Multiply(A=@n50, B=@n51) {constB=0.15}), B=NamedRerouteUsage())
  NamedRerouteDeclaration(None=Saturate(None=Subtract(A=@n45, B=@n52))) {name=Foam Mask, nodeColor=<rgba 1 0 0.4219 1>}
  n53 = VectorParameter() {parameterName=Emissive Morning, defaultValue=<rgba 0 0.0365 0.009 1>}
  n54 = VectorParameter() {parameterName=Emissive Day, defaultValue=<rgba 0 0.0081 0.0104 1>}
  n55 = VectorParameter() {parameterName=Emissive Night, defaultValue=<rgba 0 0.097 0.1354 1>}
  n56 = Multiply(A=NamedRerouteUsage()) {constB=2}
  n57 = NamedRerouteUsage()
  n58 = ScalarParameter() {defaultValue=0.15, parameterName=Emission Offset}
  n59 = Multiply(
   A=Multiply(
    A=MaterialFunctionCall(Color_Morning=@n53.RGB, Color_Day=@n54.RGB, Color_Evening=@n53.RGB, Color_Night=@n55.RGB) {materialFunction=&Color_Time_of_Day}.Result
    B=Add(A=@n56)
   )
   B=Add(A=Multiply(A=@n57, B=@n58))
  )
  n60 = NamedRerouteUsage()
  NamedRerouteDeclaration(None=EyeAdaptationInverse(LightValueInput=Multiply(A=@n59, B=@n60), AlphaInput=ScalarParameter() {defaultValue=0.5, parameterName=EyeAdaptationInverse}).EyeAdaptationInverse) {name=Emission, nodeColor=<rgba 1 0.1172 0 1>}
  n61 = Divide(A=WorldPosition().XY, B=ScalarParameter() {defaultValue=5000, parameterName=Offset Scale}) {constB=800}
  n62 = Time()
  n63 = Add(A=@n61, B=Multiply(A=@n62, B=Constant2Vector() {g=0.012, r=0.01}) {constB=0.01})
  n64 = Multiply(A=@n61) {constB=0.83}
  n65 = Multiply(A=@n62, B=Constant2Vector() {g=-0.008, r=-0.01}) {constB=0.01}
  n66 = ComponentMask(None=TextureSample(UVs=Add(A=@n64, B=@n65)) {samplerType=SAMPLERTYPE_LinearColor, texture=&/T_Water_Hermes_Bugs_Offset}.RGBA) {a=true, b=true}
  n67 = ScalarParameter() {defaultValue=0.08, parameterName=Offset Distortion}
  n68 = QualitySwitch(Default=@n63, Low=@n63, High=@n63, Medium=@n63, Epic=Add(A=@n63, B=Multiply(A=@n66, B=@n67) {constB=0.1}))
  n69 = Multiply(
   A=NamedRerouteDeclaration(None=TextureSample(UVs=@n68) {samplerType=SAMPLERTYPE_LinearColor, texture=&T_Water_Hermes_Bugs_Offset}.RGBA) {name=Offset Texture, nodeColor=<rgba 1 0 0.8438 1>}
   B=ScalarParameter() {defaultValue=1.7, parameterName=Offset Pattern intensity}
  ) {constB=2}
  n70 = Multiply(A=NamedRerouteUsage(), B=ScalarParameter() {defaultValue=2, parameterName=Offset Edge intensity})
  n71 = Time()
  n72 = VectorParameter() {parameterName=Offset Speed, defaultValue=<rgba 0.1 0.25 0.45 0.55>}
  n73 = AppendVector(A=@n72.RGB, B=@n72.RGB)
  MaterialExpressionSine_1 = Sine(None=Add(A=Subtract(A=@n69, B=@n70), B=Multiply(A=@n71, B=@n73) {constB=0.2}))
  n74 = Convert(None=@MaterialExpressionSine_1) {
   nodeName=Break Float3
   convertInputs [{
     type=Vector3
     expressionInput {
      expression=&self:MaterialExpressionSine_1
      inputName=None
      mask=0
      maskA=0
      maskB=0
      maskG=0
      maskR=0
      outputIndex=0
     }
    }]
   convertOutputs [{type=Scalar}, {type=Scalar}]
  }
  NamedRerouteDeclaration(None=Add(A=@n74, B=@n74)) {name=Offset Slow Pattern, nodeColor=<rgba 1 0.4219 0 1>}
  NamedRerouteDeclaration(None=Add(A=ComponentMask(None=@MaterialExpressionSine_1) {a=true, b=true}, B=ComponentMask(None=@MaterialExpressionSine_1) {r=true})) {name=Offset Normal Pattern, nodeColor=<rgba 0 1 0.4609 1>}
  n75 = DeriveNormalZ(InXY=Multiply(A=NamedRerouteUsage(), B=ScalarParameter() {defaultValue=0.01, parameterName=Normal Intensity}) {constB=0.05})
  n76 = LinearInterpolate(
   A=ScalarParameter() {defaultValue=0.5, parameterName=Refraction Intensity Underwater}
   B=ScalarParameter() {defaultValue=5, parameterName=Refraction Intensity}
   Alpha=Saturate(None=TwoSidedSign())
  ) {constA=1}
  NamedRerouteDeclaration(None=Add(A=Multiply(A=@n75, B=@n76) {constB=0.3})) {name=Refraction, nodeColor=<rgba 0.4297 1 0 1>}
  n77 = Multiply(A=@n13) {constB=5}
  n78 = Multiply(A=Multiply(A=Multiply(A=@n16, B=@n77), B=NamedRerouteUsage()), B=NamedRerouteUsage())
  n79 = Constant()
  n80 = Multiply(A=StaticSwitchParameter(True=Constant(), False=AppendVector(A=@n78, B=@n79)) {parameterName=HLOD}, B=@n25)
  NamedRerouteDeclaration(None=QualitySwitch(Default=@n1, Low=@n1, High=@n1, Medium=@n1, Epic=Multiply(A=@n80, B=@n28))) {name=Bubbles Normals, nodeColor=<rgba 0.5234 1 0 1>}
  NamedRerouteDeclaration(None=EyeAdaptationInverse(
    LightValueInput=VectorParameter() {parameterName=Foam Color, defaultValue=<rgba 0.0222 0.4005 0.4271 1>}.RGB
    AlphaInput=ScalarParameter() {defaultValue=0.5, parameterName=EyeAdaptationInverse}
   ).EyeAdaptationInverse) {name=Foam Color, nodeColor=<rgba 1 0.2344 0 1>}
  n81 = VectorParameter() {parameterName=Fallback Morning, defaultValue=<rgba 0.0024 0.2143 0.2292 1>}
  n82 = VectorParameter() {parameterName=Fallback Day, defaultValue=<rgba 0.0018 0.1607 0.1719 1>}
  n83 = VectorParameter() {parameterName=Fallback Night, defaultValue=<rgba 0.0035 0.3165 0.3385 1>}
  n84 = Add(
   A=LinearInterpolate(
    A=MaterialFunctionCall(Color_Morning=@n81.RGB, Color_Day=@n82.RGB, Color_Evening=@n81.RGB, Color_Night=@n83.RGB) {materialFunction=&Color_Time_of_Day}.Result
    B=VectorParameter() {parameterName=Fallback Edge, defaultValue=<rgba 0.0241 0.1406 0.1023 1>}.RGB
    Alpha=NamedRerouteUsage()
   )
   B=Multiply(A=NamedRerouteUsage(), B=ScalarParameter() {defaultValue=0.01, parameterName=Fallback Offset}) {constB=0.1}
  )
  n85 = NamedRerouteUsage()
  NamedRerouteDeclaration(None=EyeAdaptationInverse(LightValueInput=Subtract(A=@n84, B=@n85), AlphaInput=ScalarParameter() {defaultValue=0.5, parameterName=EyeAdaptationInverse}).EyeAdaptationInverse) {name=FallBack Color, nodeColor=<rgba 1 0.2344 0 1>}
  n86 = Constant()
  n87 = SetMaterialAttributes!(
   MaterialAttributes
   Base Color=VectorParameter() {parameterName=Underwater Color, defaultValue=<rgba 0 0.7998 1 1>}.RGB
   Specular=@n86
   Roughness=@n86
   Opacity=ScalarParameter() {defaultValue=0.1, parameterName=Underwater Opacity}
  ) {attributeSetTypes [
    #1
    #2
    #3
    #4
   ]}
  n88 = Normalize(VectorInput=Add(A=@n75, B=NamedRerouteUsage()))
  n89 = NamedRerouteUsage()
  n90 = NamedRerouteUsage()
  n91 = Multiply(A=@n90, B=@n90)
  n92 = SetMaterialAttributes!(
   MaterialAttributes
   Base Color=NamedRerouteUsage()
   Specular=Multiply(A=ScalarParameter() {defaultValue=0.7, parameterName=Water Specular}, B=NamedRerouteUsage())
   Roughness=ScalarParameter() {defaultValue=0.1, parameterName=Water Roughness}
   Normal=NamedRerouteDeclaration(None=LinearInterpolate(A=@n88, B=@n89, Alpha=@n91)) {name=Water Normals, nodeColor=<rgba 1 0 0.7266 1>}
   Opacity=Constant()
  ) {attributeSetTypes [
    #1
    #2
    #3
    #5
    #4
   ]}
  n93 = SetMaterialAttributes(
   Base Color=NamedRerouteUsage()
   Specular=ScalarParameter() {defaultValue=0.7, parameterName=Water Specular}
   Roughness=ScalarParameter() {defaultValue=0.1, parameterName=Water Roughness}
   Normal=NamedRerouteUsage()
   Opacity=Constant() {r=0.7}
  ) {attributeSetTypes [
    #1
    #2
    #3
    #5
    #4
   ]}
  n94 = NamedRerouteDeclaration(None=@n93) {name=Far Material, nodeColor=<rgba 1 0 0.6562 1>}
  n95 = Multiply(A=NamedRerouteUsage(), B=NamedRerouteUsage())
  n96 = Constant3Vector() {constant=<rgba 0 0 1 0>}
  n97 = TextureSample(UVs=NamedRerouteUsage()) {samplerType=SAMPLERTYPE_Normal, texture=&/T_Water_Hermes_Foam_N}
  n98 = SetMaterialAttributes(
   Base Color=NamedRerouteUsage()
   Specular=Multiply(A=ScalarParameter() {defaultValue=0.4, parameterName=Foam Specular}, B=NamedRerouteUsage())
   Roughness=ScalarParameter() {defaultValue=1, parameterName=Foam Roughness}
   Normal=NamedRerouteDeclaration(None=QualitySwitch(Default=@n96, Low=@n96, High=@n97.RGB, Medium=@n96, Epic=@n97.RGB)) {name=Foam Normals, nodeColor=<rgba 0.9453 1 0 1>}
   Opacity=Constant() {r=0.7}
  ) {attributeSetTypes [
    #1
    #2
    #3
    #5
    #4
   ]}
  n99 = NamedRerouteUsage()
  n100 = GetMaterialAttributes(None=BlendMaterialAttributes(A=BlendMaterialAttributes(A=@n92, B=@n94, Alpha=@n95), B=@n98, Alpha=Multiply(A=@n99, B=@n99) {constB=0.75})) {attributeGetTypes [#2]}
  n101 = SetMaterialAttributes!(
   MaterialAttributes=@n100.MaterialAttributes
   Specular=QualitySwitch(
    Default=@n100.Specular
    Low=Multiply(A=@n100.Specular) {constB=0.5}
    High=@n100.Specular
    Medium=Multiply(A=@n100.Specular) {constB=0.8}
    Epic=@n100.Specular
   )
  ) {attributeSetTypes [#2]}
  n102 = Saturate(None=TwoSidedSign())
  n103 = NamedRerouteUsage()
  SetMaterialAttributes!(
   MaterialAttributes=StaticSwitchParameter(True=NamedRerouteUsage(), False=BlendMaterialAttributes(A=@n87, B=@n101, Alpha=@n102)) {parameterName=HLOD}
   World Position Offset=NamedRerouteUsage()
   Refraction=NamedRerouteUsage()
   Opacity Mask=Saturate(None=Add(A=@n103) {constB=100})
   Emissive Color=NamedRerouteUsage()
  ) {attributeSetTypes [
    #6
    #7
    #8
    #9
   ]}
  n104 = VectorParameter() {parameterName=Scattering Color, defaultValue=<rgba 0 0.8355 1 1>}
  n105 = Multiply(A=NamedRerouteUsage(), B=ScalarParameter() {defaultValue=0.03, parameterName=Scattering Offset}) {constB=0.1}
  n106 = Divide(A=Subtract(A=Add(A=@n104.RGB, B=@n105), B=NamedRerouteUsage()), B=ScalarParameter() {defaultValue=200, parameterName=Scattering Coefficient})
  n107 = ScalarParameter() {defaultValue=0.5, parameterName=EyeAdaptationInverse}
  n108 = TwoSidedSign()
  n109 = VectorParameter() {parameterName=Absorption Color, defaultValue=<rgba 1 0.1929 0.1164 1>}
  n110 = Multiply(A=NamedRerouteUsage(), B=ScalarParameter() {defaultValue=0.03, parameterName=Absorption Offset}) {constB=0.1}
  n111 = Divide(
   A=Subtract(A=Subtract(A=@n109.RGB, B=@n110), B=NamedRerouteUsage())
   B=ScalarParameter() {defaultValue=40, parameterName=Absorption Coefficient}
  )
  n112 = ScalarParameter() {defaultValue=0.5, parameterName=EyeAdaptationInverse}
  n113 = TwoSidedSign()
  n114 = TwoSidedSign()
  SingleLayerWaterMaterialOutput!(
   ScatteringCoefficients=Multiply(A=EyeAdaptationInverse(LightValueInput=@n106, AlphaInput=@n107).EyeAdaptationInverse, B=Saturate(None=@n108))
   AbsorptionCoefficients=Multiply(A=EyeAdaptationInverse(LightValueInput=@n111, AlphaInput=@n112).EyeAdaptationInverse, B=Saturate(None=@n113))
   PhaseG=ScalarParameter() {defaultValue=-0.2, parameterName=Phase}
   ColorScaleBehindWater=LinearInterpolate(
    A=ScalarParameter() {defaultValue=1.8711, parameterName=Color Scale Underwater}
    B=ScalarParameter() {defaultValue=1.8711, parameterName=Color Scale}
    Alpha=Saturate(None=@n114)
   )
  )
 ]
}
